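# Physics asset for the necklace on the Weeping Woods Cheddar medium male character body, defining capsule bodies for the dangling necklace and upper spine.
# Three necklace bones chain from a tiny master capsule through constraints with limited swing and twist, loosest on the second swing axis.
# The master attaches to spine_05, which links to spine_04, both through free angular constraints.
# Large spine capsules approximate the torso alongside the chain.
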PhysicsAsset /BRCosmetics/Characters/Player/Male/Medium/Bodies/M_MED_WeepingWoods_Cheddar/Meshes/M_MED_Weeping_Woods_Cheddar_necklace_Physics
bodies {
 dyn_necklace_01 [{shapeType=Capsule, length=16.2521, radius=3.3786, center=<4.07 0 0>, rotation=<rot 0 0 89.9994>}]
 dyn_necklace_01_MSTR [{shapeType=Capsule, length=0.01, radius=0.01}]
 dyn_necklace_02 [{shapeType=Capsule, length=5, radius=3, center=<5 0 0.731>, rotation=<rot 90 0 0>}]
 dyn_necklace_03 [{shapeType=Capsule, length=5, radius=3, center=<5 0 0.3195>, rotation=<rot 90 0 0>}]
 spine_04 [{shapeType=Capsule, length=11.9808, radius=17.0818, center=<0 2.6092 0>}]
 spine_05 [{shapeType=Capsule, length=20.9808, radius=8.0818, center=<3.6954 2.4718 0>}]
}
constraints [
 {bone1Name=dyn_necklace_01, bone2Name=dyn_necklace_01_MSTR, swing1LimitDegrees=5, swing2LimitDegrees=25, twistLimitDegrees=5}
 {bone1Name=dyn_necklace_02, bone2Name=dyn_necklace_01, swing1LimitDegrees=5, swing2LimitDegrees=25, twistLimitDegrees=5}
 {bone1Name=dyn_necklace_03, bone2Name=dyn_necklace_02, swing1LimitDegrees=5, swing2LimitDegrees=25, twistLimitDegrees=5}
 {
  bone1Name=dyn_necklace_01_MSTR
  bone2Name=spine_05
  swing1LimitDegrees=45
  swing1Motion=Free
  swing2LimitDegrees=45
  swing2Motion=Free
  twistLimitDegrees=45
  twistMotion=Free
 }
 {
  bone1Name=spine_05
  bone2Name=spine_04
  swing1LimitDegrees=45
  swing1Motion=Free
  swing2LimitDegrees=45
  swing2Motion=Free
  twistLimitDegrees=45
  twistMotion=Free
 }
]
